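; Battle pass quest gameplay ability that tracks whether the player keeps the first weapon they pick up.
; A gameplay event supplies the pickup; the first item definition matching a configured tag is stored via DoOnce, while non-matching pickups end the ability.
; A one-second looping timer scans the owner's inventory for that definition, applying a "has weapon" GameplayEffect while present.
; When it is missing from a non-empty inventory, a "lost weapon" effect applies and the timer clears.
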(event Ability|EventOnEndAbility
  (params (bWasCancelled :type Boolean)))

(event Ability|EventActivateAbilityFromEvent
  (params (EventData :type Gameplay%20Event%20Data%20Structure%20%28by%20ref%29))
  (bind (_eventtag _instigator _target) (Utilities|Struct|BreakGameplayEventData EventData))
  (bind _asfort_pickup_athena (Utilities|Casting|CastToFortPickupAthena _target))
  (bind _returnvalue (Pickup|GetPrimaryItemDefinition _asfort_pickup_athena))
  (|CallParentFunction EventData)
  (if (GameplayTags|HasMatchingGameplayTag _returnvalue (Variables|Default|GetTag))
    (Utilities|FlowControl|DoOnce false
      (:Completed
        (Variables|Default|SetFirst%20Weapon _returnvalue)
        (bind _returnvalue_2 (Utilities|Time|SetTimerbyEvent (delegate Custom|CustomEvent) 1.0 true))
        (Variables|Default|SetTimer%20Handle _returnvalue_2)))
    (else
      (Ability|EndAbility))))

(event Custom|CustomEvent
  (Variables|Default|SetHas%20First%20Weapon false)
  (bind _items (InventoryManagement|GetItemsfromInventoryOwner (Fortnite|GetFortPlayerControllerfromActor (Variables|GAVars|Data|GetOwningPlayerPawn))))
  (if (Utilities|Array|IsNotEmpty _items)
    (for _array_element _items
      (if (== (Item|GetItemDefinition _array_element) (Variables|Default|GetFirst%20Weapon))
        (Variables|Default|SetHas%20First%20Weapon true)))
    (if (Variables|Default|GetHas%20First%20Weapon)
      (Creative|GameplayEffects|SimpleApplyGameplayEffectToActor (Variables|Default|GetHas%20First%20Weapon%20GE) (Fortnite|GetFortPlayerControllerfromActor (Variables|GAVars|Data|GetOwningPlayerPawn)))
      (else
        (Creative|GameplayEffects|SimpleApplyGameplayEffectToActor (Variables|Default|GetLost%20First%20Weapon%20GE) (Fortnite|GetFortPlayerControllerfromActor (Variables|GAVars|Data|GetOwningPlayerPawn)))
        (Utilities|Time|ClearandInvalidateTimerbyHandle (Variables|Default|GetTimer%20Handle))))))
